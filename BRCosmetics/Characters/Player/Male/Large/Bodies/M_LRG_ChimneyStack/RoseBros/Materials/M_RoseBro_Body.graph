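# Default Lit material for the RoseBros body of the large male player cosmetic ChimneyStack, flagged for skeletal mesh use.
# It exposes Diffuse, SRM (mask-sampled) and Normals texture parameters with placeholder defaults, so instances can supply their own maps.
# The graph holds only these three samplers, with no math nodes and no visible links to material outputs.
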Material /BRCosmetics/Characters/Player/Male/Large/Bodies/M_LRG_ChimneyStack/RoseBros/Materials/M_RoseBro_Body
graph {
 properties {bAutomaticallySetUsageInEditor=false, bUsedWithSkeletalMesh=true, usedShadingModels=Default Lit}
 nodes [
  TextureSampleParameter2D() {parameterName=Diffuse, texture=/Game/Global/Textures/Default/Blanks/T_MidgreysRGB}
  TextureSampleParameter2D() {parameterName=SRM, samplerType=SAMPLERTYPE_Masks, texture=&T_Fortnite_Default_S}
  TextureSampleParameter2D() {parameterName=Normals, samplerType=SAMPLERTYPE_Normal, texture=&FlatNormalBlue}
 ]
}
